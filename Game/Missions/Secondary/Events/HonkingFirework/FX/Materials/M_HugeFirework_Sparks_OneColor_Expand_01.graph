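# Masked, two-sided particle material for the expanding spark streaks of the huge firework in the HonkingFirework secondary mission event.
# World position offset combines scaled UV offsets, an OffsetMesh-driven push and SplineThicken, with vertex red tapering width and LineWidth scaling the tip.
# Color blends two orange tints by the spark texture, then applies particle-driven hue shift, brightness and desaturation.
# The mask adds panning noise inside a PanSmoke-positioned sphere mask to the spark texture, scaled by particle alpha.
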Material /Game/Missions/Secondary/Events/HonkingFirework/FX/Materials/M_HugeFirework_Sparks_OneColor_Expand_01
graph {
 properties {
  bEnableResponsiveAA=true
  bUsedWithMeshParticles=true
  bUsedWithParticleSprites=true
  blendMode=BLEND_Masked
  pixelDepthOffsetMode=PDOM_Legacy
  twoSided=true
  usedShadingModels=Default Lit
 }
 nodes [
  n1 = MaterialFunctionCall(Float2=Subtract(A=TextureCoordinate()) {constB=0.5}) {materialFunction=&BreakOutFloat2Components}
  n2 = MaterialFunctionCall(Float2=VectorParameter() {parameterName=XY Scale, defaultValue=<rgba 1 1 0 0>}.RGB) {materialFunction=&BreakOutFloat2Components}
  n3 = AppendVector(A=AppendVector(A=Multiply(A=@n1.R, B=@n2.R), B=Multiply(A=@n1.G, B=@n2.G)), B=Constant() {r=1})
  n4 = Transform(None=VectorParameter() {parameterName=OffsetDistance, defaultValue=<rgba 0 0 1 0>}.RGB)
  n5 = LinearInterpolate(Alpha=DynamicParameter() {defaultValue=<rgba 1 1 -1 1>, paramNames [Test,PanSmoke,OffsetMesh,LineWidth]}.OffsetMesh) {constB=150}
  n6 = Clamp(None=VertexColor().R) {minDefault=0.5}
  n7 = ScalarParameter() {defaultValue=5, parameterName=WidthBase}
  n8 = Multiply(
   A=ScalarParameter() {defaultValue=50, parameterName=WidthTip}
   B=DynamicParameter() {defaultValue=<rgba 1 1 -1 1>, paramNames [Test,PanSmoke,OffsetMesh,LineWidth]}.LineWidth
  )
  n9 = Clamp(None=VertexColor().R) {minDefault=0.5}
  Add(
   A=Add(A=Transform(None=@n3), B=Multiply(A=@n4, B=@n5))
   B=MaterialFunctionCall(WidthBase=Multiply(A=@n6, B=@n7), WidthTip=Multiply(A=@n8, B=@n9)) {materialFunction=&SplineThicken}.WorldPositionOffset
  )
  n10 = ParticleColor()
  n11 = VertexColor()
  n12 = TextureSample(UVs=TextureCoordinate()) {texture=/Game/Missions/Secondary/Events/HonkingFirework/FX/Textures/T_FireWorks_Capsule_Sparks_01}
  n13 = LinearInterpolate(
   A=VectorParameter() {parameterName=SparkColor0, defaultValue=<rgba 1 0.0978 0 0>}.RGB
   B=VectorParameter() {parameterName=SparkColor1, defaultValue=<rgba 1 0.2289 0 0>}.RGB
   Alpha=@n12.G
  )
  n14 = MaterialFunctionCall(LDR Input=Clamp(None=Multiply(A=@n11.G, B=@n13)), HDR Tint=Multiply(A=@n11.G, B=@n13)) {materialFunction=&DeriveHDRfromLDR}
  Desaturation(
   None=Multiply(
    A=ParticleColor().G
    B=MaterialFunctionCall(Hue Shift Percentage=@n10.R, Texture=@n14.RemappedRange) {materialFunction=&HueShift}.Result
   )
   Fraction=ScalarParameter() {defaultValue=0.5, parameterName=Fraction}
  )
  Multiply(A=@n10.R, B=Clamp(None=VertexColor().B) {maxDefault=0.75, minDefault=0.25})
  n15 = VertexColor()
  n16 = Multiply(A=@n12.G) {constB=0.5}
  n17 = AppendVector(A=Constant(), B=DynamicParameter() {defaultValue=<rgba 1 1 -1 1>, paramNames [Test,PanSmoke,OffsetMesh,LineWidth]}.PanSmoke)
  n18 = Clamp(None=SphereMask(
    A=TextureCoordinate() {uTiling=0}
    B=ComponentMask(None=@n17) {g=true, r=true}
    Radius=ScalarParameter() {defaultValue=0.7, parameterName=GlowTip_Radius_1}
    Hardness=ScalarParameter() {parameterName=GlowTip_Hardness_1}
   ))
  n19 = TextureCoordinate() {uTiling=0.25, vTiling=2}
  n20 = TextureCoordinate() {uTiling=0.25, vTiling=2}
  n21 = Multiply(
   A=TextureSample(UVs=Panner(Coordinate=@n19) {speedY=0.1}) {texture=/Game/Global/Textures/Default/Noise/T_TileNoise_02}.G
   B=TextureSample(UVs=Panner(Coordinate=@n20) {speedY=0.02}) {texture=&T_TileNoise_02}.G
  )
  n22 = Clamp(None=Power(Base=Multiply(A=@n21) {constB=150}) {constExponent=1.25})
  Multiply(A=Multiply(A=Clamp(None=@n15.G) {minDefault=0.5}, B=ParticleColor().A), B=Add(A=Multiply(A=@n16), B=Multiply(A=@n18, B=@n22)))
 ]
}
